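; Fortnite ceiling prop, such as a hanging light, whose fixture hangs at an adjustable FixtureLength below its base.
; The construction script measures mesh bounds, positions the fixture and socketed light, then either stretches the down rod with matching UV scale or stacks instanced segments.
; It optionally swaps mesh collision for a capsule and applies building texture data tint or override material.
; Down-rod materials are exposed as allowed UEFN references.
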(event EventBeginPlay)

(event Collision|EventActorBeginOverlap
  (params (OtherActor :type Actor%20Object%20Reference)))

(event EventTick
  (params (DeltaSeconds :type Float%20%28single-precision%29)))

(fn AddCapsuleCollision
  (bind _returnvalue (AddComponent|NODE_AddCapsuleComponent-3 :RelativeTransform (Math|Transform|MakeTransform (Math|Vector|MakeVector 0.0 0.0 (- (/ (Variables|AdjustableLight|GetFixtureLength) 2.0) (- (Variables|AdjustableLight|GetFixtureLength) (Variables|AdjustableLight|GetBaseZBounds)))))))
  (bind _output_get (Variables|Default|SetCapsuleCollision _returnvalue))
  (Components|Capsule|SetCapsuleHalfHeight _output_get (/ (Variables|AdjustableLight|GetFixtureLength) 2.0) true)
  (for _array_element (Utilities|Array|MakeArray (Variables|AdjustableLight|GetLightFixture) (Variables|AdjustableLight|GetDownRod) (Variables|Debug|GetStaticMeshComponent))
    (Collision|SetCollisionResponsetoAllChannels _array_element "ECR_Ignore")
    (Collision|SetCollisionEnabled _array_element "NoCollision")))

(fn BuildExposedAssetProviderVersion(Implementation)
  (params (Version :type Fort%20Exposed%20Asset%20Provider%20Version%20Structure%20%28by%20ref%29))
  (|CallParentFunction Version)
  (ExposedAssetProvider|AddVersion Version 2)
  (return))

(fn CollectAssetReferencestoAllow(Implementation)
  (params (Collector :type Fort%20Exposed%20Asset%20Reference%20Collector%20Structure%20%28by%20ref%29))
  (|CallParentFunction Collector)
  (bind _returnvalue (Fortnite|Utilities|ResolveComponentFromActor 0 (Variables|AdjustableLight|GetDownRod)))
  (for _array_element (Rendering|Material|GetMaterials _returnvalue)
    (ExposedAssetProvider|CollectIndirectAssetReference Collector (Utilities|ResolveSoftReference _array_element)) ; Collect the texture data to allow it to be referenced in UEFN
  )
  (return))

(fn EnableCollisionOnMeshParts
  (for _array_element (Utilities|Array|MakeArray (Variables|AdjustableLight|GetDownRod) (Variables|AdjustableLight|GetLightFixture))
    (Collision|SetCollisionProfileName _array_element "Default" true :!class /Script/Engine.PrimitiveComponent)
    (Development|PrintString (Utilities|String|ToString(Name) (Collision|GetCollisionProfileName (stale _array_element_2)))))
  (for _array_element_2 (Utilities|Array|MakeArray (Variables|AdjustableLight|GetDownRod) (Variables|AdjustableLight|GetLightFixture))
    _))

(fn GetCurieCollisionComponentOverride
  (return (Variables|Default|GetCapsuleCollision)))

(fn GetMaterialFromRootMesh
  (Rendering|Material|SetMaterial (Variables|AdjustableLight|GetLightFixture) 0 (Material|GetMID (Variables|Debug|GetStaticMeshComponent))))

(fn GetMeshBounds
  (Variables|AdjustableLight|SetBaseZBounds (.z (Components|StaticMesh|GetLocalBounds (Variables|Debug|GetStaticMeshComponent))))
  (Variables|AdjustableLight|SetFixtureZBounds (.z (Components|StaticMesh|GetLocalBounds (Variables|AdjustableLight|GetLightFixture))))
  (Variables|AdjustableLight|SetRodZBounds (+ (.z (Components|StaticMesh|GetLocalBounds (Variables|AdjustableLight|GetDownRod))) (* (Variables|AdjustableLight|GetRodMeshOffset) -1.0))))

(fn GetTintColorFromTextureData
  (Variables|Default|SetHasMaterialOverride false)
  (Variables|Default|SetOverride%20Material 0)
  (bind _outtexturedataarray (Building|GetBuildingTextureData))
  (Variables|Default|SetHasMaterialOverride (Utilities|IsValid (Variables|/Script/FortniteGame.BuildingTextureData|GetOverrideMaterial (Utilities|Array|Get(acopy) _outtexturedataarray)) :!class /Script/Engine.KismetSystemLibrary))
  (if (Utilities|IsValid (Variables|/Script/FortniteGame.BuildingTextureData|GetOverrideMaterial (Utilities|Array|Get(acopy) _outtexturedataarray)) :!class /Script/Engine.KismetSystemLibrary)
    (Variables|Default|SetOverride%20Material (Variables|/Script/FortniteGame.BuildingTextureData|GetOverrideMaterial (Utilities|Array|Get(acopy) _outtexturedataarray))))
  (Variables|AdjustableLight|SetTintColor (Math|Conversions|ToLinearColor(Color) (Variables|/Script/FortniteGame.BuildingTextureData|GetTintColor (Utilities|Array|Get(acopy) _outtexturedataarray)))))

(fn SetDownRodMaterial
  (bind _downrod (Variables|AdjustableLight|GetDownRod))
  (bind _returnvalue (Rendering|Material|GetMaterial _downrod 0))
  (bind _downrod_2 (Variables|AdjustableLight|GetDownRod))
  (bind _returnvalue_2 (Rendering|Material|GetMaterial _downrod_2 0))
  (if (Variables|AdjustableLight|GetUseGeoForDownrod)
    (if (Utilities|IsValid _returnvalue_2 :!class /Script/Engine.KismetSystemLibrary)
      (Rendering|Material|CreateDynamicMaterialInstance (Variables|AdjustableLight|GetISM_DownRod) 0 _returnvalue_2 "None" :!class /Script/Engine.PrimitiveComponent))
    (elif (Utilities|IsValid _returnvalue :!class /Script/Engine.KismetSystemLibrary)
      (bind _returnvalue_3 (Rendering|Material|CreateDynamicMaterialInstance (Variables|AdjustableLight|GetDownRod) 0 _returnvalue "None" :!class /Script/Engine.PrimitiveComponent))
      (Rendering|Material|SetScalarParameterValue _returnvalue_3 (select (Variables|Default|GetUseHorizontalTilingUVs) "UV Scale Horizontal" "UV Scale Vertical") (Variables|AdjustableLight|GetRodScale) :!class /Script/Engine.MaterialInstanceDynamic))))

(fn SetFixturePosition
  (Transformation|SetRelativeLocation (Variables|AdjustableLight|GetLightFixture) (Math|Vector|MakeVector 0.0 0.0 (- (* (Variables|AdjustableLight|GetFixtureLength) -1.0) (Variables|AdjustableLight|GetFixtureZBounds))) false false))

(fn SetLightPosition
  (bind _returnvalue (Actor|GetComponentsByClass "/Script/Engine.LightComponent"))
  (if (> (Utilities|Array|Length _returnvalue) 0)
    (Transformation|SetWorldTransform (Utilities|Array|Get(acopy) _returnvalue) (Transformation|GetSocketTransform (Variables|AdjustableLight|GetLightFixture) (Utilities|Array|Get(acopy) (Transformation|GetAllSocketNames (Variables|AdjustableLight|GetLightFixture))) "RTS_World") false false)))

(fn SetNumberofRodMeshes
  (bind _downrod (Variables|AdjustableLight|GetDownRod))
  (bind _staticmesh (Variables|/Script/Engine.StaticMeshComponent|GetStaticMesh _downrod))
  (bind _returnvalue (AddComponent|NODE_AddInstancedStaticMeshComponent-0))
  (bind _output_get (Variables|AdjustableLight|SetISM_DownRod _returnvalue))
  (if (Utilities|IsValid _staticmesh :!class /Script/Engine.KismetSystemLibrary)
    (Components|StaticMesh|SetStaticMesh _output_get _staticmesh)
    (bind _output_get_2 (LocalVariables|SetNumberOfInstances (/ (Utilities|Operators|Subtract (Variables|AdjustableLight|GetFixtureLength) (* (Variables|AdjustableLight|GetBaseZBounds) -1.0) (* (Variables|AdjustableLight|GetFixtureZBounds) -1.0)) (* (Variables|AdjustableLight|GetRodZBounds) -1.0))))
    (for _index (range (+ (Math|Float|Floor (- _output_get_2 1.0)) 1))
      (Components|InstancedStaticMesh|AddInstance _output_get (Math|Transform|MakeTransform (Math|Vector|MakeVector 0.0 0.0 (+ (* _index (Variables|AdjustableLight|GetRodZBounds)) (Variables|AdjustableLight|GetBaseZBounds))) (Math|Rotator|MakeRotator 0.0 0.0 (Math|Conversions|ToFloat(Integer) (* _index (Variables|AdjustableLight|GetInstanceRotationAmount))))) false)
      (Rendering|SetVisibility (Variables|AdjustableLight|GetDownRod) false false)
      (Variables|Default|SetFixtureOffsetAdjustment (* (- (- _output_get_2 1.0) (Math|Float|Floor (- _output_get_2 1.0))) (Variables|AdjustableLight|GetRodZBounds))))
    (else
      (Development|PrintString "DownRod mesh not found"))))

(fn SetRodPosition
  (bind _rodzbounds (Variables|AdjustableLight|GetRodZBounds))
  (bind _returnvalue (* _rodzbounds -1.0))
  (if (!= _returnvalue 0)
    (bind _output_get (Variables|AdjustableLight|SetRodScale (/ (Utilities|Operators|Subtract (Variables|AdjustableLight|GetFixtureLength) (* (Variables|AdjustableLight|GetBaseZBounds) -1.0) (* (Variables|AdjustableLight|GetFixtureZBounds) -1.0)) _returnvalue)))
    (Transformation|SetRelativeScale3D (Variables|AdjustableLight|GetDownRod) (Math|Vector|MakeVector 1.0 1.0 _output_get))
    (Transformation|SetRelativeLocation (Variables|AdjustableLight|GetDownRod) (Math|Vector|MakeVector 0.0 0.0 (Variables|AdjustableLight|GetBaseZBounds)) false false)))

(fn SetTintColorOnMeshParts
  (bind _lightfixture (Variables|AdjustableLight|GetLightFixture))
  (if (Utilities|IsValid (Variables|Default|GetOverride%20Material) :!class /Script/Engine.KismetSystemLibrary)
    (if (Variables|Default|GetApplyOverrideMaterialToDownRod)
      (for _array_element (Utilities|Array|MakeArray (Variables|AdjustableLight|GetDownRod) _lightfixture)
        (Rendering|Material|SetMaterial _array_element 0 (Variables|Default|GetOverride%20Material)))
      (else
        (for _array_element_2 (Utilities|Array|MakeArray _lightfixture)
          (Default|SetColoronSecondaryMesh (Variables|AdjustableLight|GetTintColor) _array_element_2 :!label _label))))
    (else
      (for _array_element_3 (Utilities|Array|MakeArray (Variables|AdjustableLight|GetDownRod) (Variables|AdjustableLight|GetLightFixture))
        (goto _label))))
  (goto _label))

(fn ConstructionScript
  (bind _returnvalue_location (Math|Transform|BreakTransform (Transformation|GetRelativeTransform (Variables|AdjustableLight|GetLightFixture))))
  (CallFunction|GetMeshBounds)
  (if (!= (Variables|AdjustableLight|GetUseMeshCollision) true)
    (CallFunction|AddCapsuleCollision))
  (CallFunction|SetFixturePosition)
  (CallFunction|SetLightPosition)
  (if (Variables|AdjustableLight|GetUseGeoForDownrod)
    (CallFunction|SetNumberofRodMeshes)
    (else
      (CallFunction|SetRodPosition)))
  (CallFunction|ValidateBuildingTextureData)
  (if (Variables|AdjustableLight|GetHasTextureData)
    (CallFunction|GetTintColorfromTextureData)
    (else
      (CallFunction|GetMaterialfromRootMesh)))
  (CallFunction|SetDownRodMaterial)
  (CallFunction|SetTintColoronMeshParts)
  (Transformation|SetRelativeTransform (Variables|AdjustableLight|GetLightFixture) (Math|Transform|MakeTransform (Math|Vector|MakeVector (.x (.location (Transformation|GetRelativeTransform (Variables|AdjustableLight|GetLightFixture)))) (.y (.location (Transformation|GetRelativeTransform (Variables|AdjustableLight|GetLightFixture)))) (- (.z (.location (Transformation|GetRelativeTransform (Variables|AdjustableLight|GetLightFixture)))) (Variables|Default|GetFixtureOffsetAdjustment))) (.rotation (Transformation|GetRelativeTransform (Variables|AdjustableLight|GetLightFixture))) (.scale (Transformation|GetRelativeTransform (Variables|AdjustableLight|GetLightFixture)))) false false)
  (Rendering|Material|SetDefaultCustomPrimitiveDataVector3 (Variables|AdjustableLight|GetLightFixture) 0 (* _returnvalue_location -1.0))
  (Rendering|Material|SetDefaultCustomPrimitiveDataFloat (Variables|AdjustableLight|GetLightFixture) 4 (Variables|AdjustableLight|GetFixtureLength))
  ; IF TD Color Bool is true , Set TD as tint for each elements
  (if (Variables|Default|GetUse%20TD%20for%20Tint)
    (bind _outtexturedataarray (Building|GetBuildingTextureData))
    (if (> (Utilities|Array|Length _outtexturedataarray) 0)
      (for _array_element (Components|GetChildrenComponents (Variables|AdjustableLight|GetLightFixture) false)
        (bind _asstatic_mesh_component (Utilities|Casting|CastToStaticMeshComponent _array_element))
        (Default|SetColoronSecondaryMesh (Math|Conversions|ToLinearColor(Color) (Variables|/Script/FortniteGame.BuildingTextureData|GetTintColor (Utilities|Array|Get(acopy) _outtexturedataarray))) _asstatic_mesh_component)))))

(fn ValidateBuildingTextureData
  (bind _outtexturedataarray (Building|GetBuildingTextureData))
  (Variables|AdjustableLight|SetHasTextureData (> (Utilities|Array|Length _outtexturedataarray) 0))
  (return))
