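; Fortnite building prop for a height-adjustable floor light, assembled from base, rod and fixture meshes in the construction script.
; Mesh bounds and FixtureLength position the fixture; the first light component follows its first socket.
; The rod either stretches with compensating UV tiling or repeats as rotated instanced links.
; An optional capsule replaces mesh collision; building texture data supplies tint, otherwise the fixture takes the base material, and emissive color is applied.
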(event EventBeginPlay)

(event Collision|EventActorBeginOverlap
  (params (OtherActor :type Actor%20Object%20Reference)))

(event EventTick
  (params (DeltaSeconds :type Float%20%28single-precision%29)))

(fn AddCapsuleCollision
  (bind _returnvalue (AddComponent|NODE_AddCapsuleComponent-3 :RelativeTransform (Math|Transform|MakeTransform (Math|Vector|MakeVector 0.0 0.0 (/ (Variables|AdjustableLight|GetFixtureLength) 2.0)))))
  (bind _output_get (Variables|Default|SetCapsuleCollision _returnvalue))
  (Components|Capsule|SetCapsuleHalfHeight _output_get (/ (Variables|AdjustableLight|GetFixtureLength) 2.0) true)
  (for _array_element (Utilities|Array|MakeArray (Variables|AdjustableLight|GetLightFixture) (Variables|AdjustableLight|GetUpRod) (Variables|Debug|GetStaticMeshComponent))
    (Collision|SetCollisionEnabled _array_element "NoCollision")))

(fn EnableCollisionOnMeshParts
  (for _array_element (Utilities|Array|MakeArray (Variables|AdjustableLight|GetUpRod) (Variables|AdjustableLight|GetLightFixture))
    (Collision|SetCollisionProfileName _array_element "Default" true :!class /Script/Engine.PrimitiveComponent)
    (Development|PrintString (Utilities|String|ToString(Name) (Collision|GetCollisionProfileName (stale _array_element_2)))))
  (for _array_element_2 (Utilities|Array|MakeArray (Variables|AdjustableLight|GetUpRod) (Variables|AdjustableLight|GetLightFixture))
    _))

(fn GetCurieCollisionComponentOverride
  (return (Variables|Default|GetCapsuleCollision)))

(fn GetMaterialFromRootMesh
  (Rendering|Material|SetMaterial (Variables|AdjustableLight|GetLightFixture) 0 (Material|GetMID (Variables|Debug|GetStaticMeshComponent))))

(fn GetMeshBounds
  (bind (_min _max) (Components|StaticMesh|GetLocalBounds (Variables|AdjustableLight|GetLightFixture)))
  (bind (_min_2 _max_2) (Components|StaticMesh|GetLocalBounds (Variables|AdjustableLight|GetUpRod)))
  (bind (_min_3 _max_3) (Components|StaticMesh|GetLocalBounds (Variables|Debug|GetStaticMeshComponent)))
  (Variables|AdjustableLight|SetBaseZBounds (+ (.z _max_3) (Variables|Default|GetBaseMeshBoundsOffset)))
  (Variables|AdjustableLight|SetFixtureZBounds (.z _max))
  (Variables|AdjustableLight|SetRodZBounds (+ (.z _max_2) (Variables|AdjustableLight|GetRodMeshOffset))))

(fn GetTintColorFromTextureData
  (bind _outtexturedataarray (Building|GetBuildingTextureData))
  (Variables|AdjustableLight|SetTintColor (Math|Conversions|ToLinearColor(Color) (Variables|/Script/FortniteGame.BuildingTextureData|GetTintColor (Utilities|Array|Get(acopy) _outtexturedataarray)))))

(fn SetDownRodMaterial
  (bind _uprod (Variables|AdjustableLight|GetUpRod))
  (bind _returnvalue (Rendering|Material|GetMaterial _uprod 0))
  (bind _uprod_2 (Variables|AdjustableLight|GetUpRod))
  (bind _returnvalue_2 (Rendering|Material|GetMaterial _uprod_2 0))
  (if (Variables|AdjustableLight|GetUseGeoForUprod)
    (if (Utilities|IsValid _returnvalue_2 :!class /Script/Engine.KismetSystemLibrary)
      (Rendering|Material|CreateDynamicMaterialInstance (Variables|AdjustableLight|GetISM_DownRod) 0 _returnvalue_2 "None" :!class /Script/Engine.PrimitiveComponent))
    (elif (Utilities|IsValid _returnvalue :!class /Script/Engine.KismetSystemLibrary)
      (bind _returnvalue_3 (Rendering|Material|CreateDynamicMaterialInstance (Variables|AdjustableLight|GetUpRod) 0 _returnvalue "None" :!class /Script/Engine.PrimitiveComponent))
      (Rendering|Material|SetScalarParameterValue _returnvalue_3 (select (Variables|Default|GetUseHorizontalTilingUVs) "UV Scale Horizontal" "UV Scale Vertical") (Variables|AdjustableLight|GetRodScale) :!class /Script/Engine.MaterialInstanceDynamic))))

(fn SetFixturePosition
  (Transformation|SetRelativeLocation (Variables|AdjustableLight|GetLightFixture) (Math|Vector|MakeVector 0.0 0.0 (- (Variables|AdjustableLight|GetFixtureLength) (Variables|AdjustableLight|GetFixtureZBounds))) false false))

(fn SetLightPosition
  (bind _returnvalue (Actor|GetComponentsByClass "/Script/Engine.LightComponent"))
  (bind _lightfixture (Variables|AdjustableLight|GetLightFixture))
  (bind _returnvalue_2 (Transformation|GetAllSocketNames _lightfixture))
  (if (> (Utilities|Array|Length _returnvalue) 0)
    (if (> (Utilities|Array|Length _returnvalue_2) 0)
      (Transformation|SetWorldTransform (Utilities|Array|Get(acopy) _returnvalue) (Math|Transform|MakeTransform (+ (Variables|AdjustableLight|GetLightOffset) (.location (Transformation|GetSocketTransform _lightfixture (Utilities|Array|Get(acopy) _returnvalue_2) "RTS_World"))) (.rotation (Transformation|GetWorldTransform (Utilities|Array|Get(acopy) _returnvalue))) (.scale (Transformation|GetSocketTransform _lightfixture (Utilities|Array|Get(acopy) _returnvalue_2) "RTS_World"))) false false))))

(fn SetNumberofRodMeshes
  (bind _uprod (Variables|AdjustableLight|GetUpRod))
  (bind _staticmesh (Variables|/Script/Engine.StaticMeshComponent|GetStaticMesh _uprod))
  (bind _returnvalue (AddComponent|NODE_AddInstancedStaticMeshComponent-0))
  (bind _output_get (Variables|AdjustableLight|SetISM_DownRod _returnvalue))
  (if (Utilities|IsValid _staticmesh :!class /Script/Engine.KismetSystemLibrary)
    (Components|StaticMesh|SetStaticMesh _output_get _staticmesh)
    (bind _output_get_2 (LocalVariables|SetNumberOfInstances (/ (Utilities|Operators|Subtract (Variables|AdjustableLight|GetFixtureLength) (* (Variables|AdjustableLight|GetBaseZBounds) -1.0) (* (Variables|AdjustableLight|GetFixtureZBounds) -1.0)) (* (Variables|AdjustableLight|GetRodZBounds) -1.0))))
    (for _index (range (+ (Math|Float|Floor _output_get_2) 1))
      (Components|InstancedStaticMesh|AddInstance _output_get (Math|Transform|MakeTransform (Math|Vector|MakeVector 0.0 0.0 (+ (* _index (Variables|AdjustableLight|GetRodZBounds)) (Variables|AdjustableLight|GetBaseZBounds))) (Math|Rotator|MakeRotator 0.0 0.0 (Math|Conversions|ToFloat(Integer) (* _index (Variables|AdjustableLight|GetChainlinkRotationAmount))))) false)
      (Rendering|SetVisibility (Variables|AdjustableLight|GetUpRod) false false))
    (else
      (Development|PrintString "DownRod mesh not found"))))

(fn SetRodPosition
  (bind _rodzbounds (Variables|AdjustableLight|GetRodZBounds))
  (if (not (Math|Float|NearlyEqual(Float) _rodzbounds))
    (bind _output_get (Variables|AdjustableLight|SetRodScale (Math|Float|Absolute(Float) (/ (Utilities|Operators|Subtract (Variables|AdjustableLight|GetFixtureLength) (Variables|AdjustableLight|GetBaseZBounds) (Variables|AdjustableLight|GetFixtureZBounds)) _rodzbounds))))
    (Transformation|SetRelativeScale3D (Variables|AdjustableLight|GetUpRod) (Math|Vector|MakeVector 1.0 1.0 _output_get))
    (Transformation|SetRelativeLocation (Variables|AdjustableLight|GetUpRod) (Math|Vector|MakeVector 0.0 0.0 (Variables|AdjustableLight|GetBaseZBounds)) false false)))

(fn SetTintColorOnMeshParts
  (for _array_element (Utilities|Array|MakeArray (Variables|AdjustableLight|GetUpRod) (Variables|AdjustableLight|GetLightFixture))
    (Default|SetColoronSecondaryMesh (Variables|AdjustableLight|GetTintColor) _array_element)))

(fn ConstructionScript
  (bind _lightfixture (Variables|AdjustableLight|GetLightFixture))
  (|CallParentFunction)
  (CallFunction|GetMeshBounds)
  (if (!= (Variables|AdjustableLight|GetUseMeshCollision) true)
    (CallFunction|AddCapsuleCollision))
  (CallFunction|SetFixturePosition)
  (CallFunction|SetLightPosition)
  (if (Variables|AdjustableLight|GetUseGeoForUprod)
    (Default|SetNumberofRodMeshes)
    (else
      (CallFunction|SetRodPosition)))
  (CallFunction|ValidateBuildingTextureData)
  (if (Variables|AdjustableLight|GetHasTextureData)
    (CallFunction|GetTintColorfromTextureData)
    (else
      (CallFunction|GetMaterialfromRootMesh)))
  (CallFunction|SetDownRodMaterial)
  (CallFunction|SetTintColoronMeshParts)
  (if (> (Rendering|Material|GetNumMaterials _lightfixture) 0)
    (Rendering|Material|SetVectorParameterValue (Class|BTechArtGlobalLibrary|GetDynamicMaterialInstance _lightfixture) "Emissive_Color_Multiplier" (Variables|00_GlobalLighting|GetEmissive%20Color%20Modifier) :!class /Script/Engine.MaterialInstanceDynamic)))

(fn ValidateBuildingTextureData
  (bind _outtexturedataarray (Building|GetBuildingTextureData))
  (Variables|AdjustableLight|SetHasTextureData (> (Utilities|Array|Length _outtexturedataarray) 0))
  (return))
